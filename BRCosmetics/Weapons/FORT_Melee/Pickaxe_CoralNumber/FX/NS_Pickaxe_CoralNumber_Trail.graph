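# Niagara swing trail for the CoralNumber pickaxe, drawing short-lived ribbons with two animation-trail emitters, each holding several ribbon renderers.
# User.Variant sets one of six style flags (Default, Blitz, Gold, Holo, Ice, Galaxy) and indexes a per-emitter color array.
# Particle color is tinted and dimmed at night through time-of-day selectors.
# Alpha fades over both particle life and emitter loop age.
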NiagaraSystem /BRCosmetics/Weapons/FORT_Melee/Pickaxe_CoralNumber/FX/NS_Pickaxe_CoralNumber_Trail
dependencies {
 usedDataInterfaces [&NiagaraDataInterfaceArrayColor, &NiagaraDataInterfaceCurve]
 usedDynamicInputs []
 usedModules [&SystemState]
}
emitters {
 Trail {
  emitterUpdateScript [
   &EmitterState()
   &FN_SpawnAnimTrail_v2(AnimTrailWidthMode=::FromSecond)
   /SetVariables(
    Emitter.BlitzEnabled:NiagaraBool=&SetBoolByIntComparison(A=&MakeIntFromFloat(Float=$User.Variant), B=1, Comparison Type=::NewEnumerator4)
    Emitter.DefaultEnabled:NiagaraBool=&SetBoolByIntComparison(A=&MakeIntFromFloat(Float=$User.Variant), B=0, Comparison Type=::NewEnumerator4)
    Emitter.GalaxyEnabled:NiagaraBool=&SetBoolByIntComparison(A=&MakeIntFromFloat(Float=$User.Variant), B=5, Comparison Type=::NewEnumerator4)
    Emitter.GoldEnabled:NiagaraBool=&SetBoolByIntComparison(A=&MakeIntFromFloat(Float=$User.Variant), B=2, Comparison Type=::NewEnumerator4)
    Emitter.HoloEnabled:NiagaraBool=&SetBoolByIntComparison(A=&MakeIntFromFloat(Float=$User.Variant), B=3, Comparison Type=::NewEnumerator4)
    Emitter.IceEnabled:NiagaraBool=&SetBoolByIntComparison(A=&MakeIntFromFloat(Float=$User.Variant), B=4, Comparison Type=::NewEnumerator4)
   ) {isSetParameters=true, module=SetVariables_#1}
  ]
  particleSpawnScript [
   /Game/Effects/Niagara/Modules/AnimTrail/FN_InitializeAnimTrail/FN_InitializeAnimTrailBoneIndecies(First Bone or Socket Index=1)
   /Game/Effects/Niagara/Modules/AnimTrail/FN_InitializeAnimTrail/FN_InitializeAnimTrail(
    AnimTrailWidth=1
    Color=&SelectLinearColorFromArray(
     Array Sampling Mode=::NewEnumerator1
     Direct Array Index=&MakeIntFromFloat(Float=$User.Variant)
     Color Selection Array=DataInterface() {propertyValues {ColorData [
        <rgba 1 1 1 0.8>
        <rgba 0.7758 0.7758 0.7758 0.5>
        <rgba 0.5 0.282 0.064 1>
        <rgba 1 1 1 0.777>
        <rgba 0.0003 0.3564 0.6795 1>
        <rgba 1 1 1 2>
       ]}}
     Random Seed=?
    )
    Lifetime=0.12
   )
  ]
  particleUpdateScript [
   &FN_UpdateAnimTrail_v2()
   &ScaleColor(
    Scale Alpha=&FloatFromCurve(
     FloatCurve=DataInterface() {propertyValues {Curve {keys [{interpMode=RCIM_Cubic, tangentMode=RCTM_User, time=0.25, value=1}, {interpMode=RCIM_Cubic, tangentMode=RCTM_User, time=1, value=0}]}}}
     Scale Curve=1.5
    )
    Scale RGB=&Multiply_VectorByFloat(
     Vector=&FN_Select_Vector_from_Time_Of_Day(Morning=<1 1 1>, Day=<1 1 1>, Evening=<1 1 1>, Night=<0.4 0.6 1>)
     Float=&FN_Select_Float_from_Time_Of_Day(Morning=1, Day=1, Evening=1, Night=0.65)
    )
    ScaleA=true
    ScaleRGB=true
   ) {module=ScaleColor002}
   &SolveForcesAndVelocity(Clamp Velocity=false, Limit Acceleration=false)
   /SetVariables() {isSetParameters=true, module=SetVariables_#2}
   &ScaleColor(
    Scale Alpha=&FloatFromCurve(
     FloatCurve=DataInterface() {propertyValues {Curve {keys [{interpMode=RCIM_Cubic, tangentMode=RCTM_User, time=0, value=1}, {interpMode=RCIM_Cubic, tangentMode=RCTM_User, time=1, value=0}]}}}
     CurveIndex=$Emitter.NormalizedLoopAge
    )
    ScaleA=true
    ScaleRGB=true
   )
  ]
  rendererClasses [&NiagaraRibbonRendererProperties]
  renderers [
   {rendererClass=&NiagaraRibbonRendererProperties}
   {rendererClass=&NiagaraRibbonRendererProperties, rendererIndex=1}
   {rendererClass=&NiagaraRibbonRendererProperties, rendererIndex=2}
   {rendererClass=&NiagaraRibbonRendererProperties, rendererIndex=3}
   {rendererClass=&NiagaraRibbonRendererProperties, rendererIndex=4}
   {rendererClass=&NiagaraRibbonRendererProperties, rendererIndex=5}
  ]
 }
 Trail001 {
  emitterUpdateScript [
   &EmitterState()
   &FN_SpawnAnimTrail_v2(AnimTrailWidthMode=::FromSecond)
   /SetVariables(
    Emitter.BlitzEnabled:NiagaraBool=&SetBoolByIntComparison(A=&MakeIntFromFloat(Float=$User.Variant), B=1, Comparison Type=::NewEnumerator4)
    Emitter.DefaultEnabled:NiagaraBool=&SetBoolByIntComparison(A=&MakeIntFromFloat(Float=$User.Variant), B=0, Comparison Type=::NewEnumerator4)
    Emitter.GalaxyEnabled:NiagaraBool=&SetBoolByIntComparison(A=&MakeIntFromFloat(Float=$User.Variant), B=5, Comparison Type=::NewEnumerator4)
    Emitter.GoldEnabled:NiagaraBool=&SetBoolByIntComparison(A=&MakeIntFromFloat(Float=$User.Variant), B=2, Comparison Type=::NewEnumerator4)
    Emitter.HoloEnabled:NiagaraBool=&SetBoolByIntComparison(A=&MakeIntFromFloat(Float=$User.Variant), B=3, Comparison Type=::NewEnumerator4)
    Emitter.IceEnabled:NiagaraBool=&SetBoolByIntComparison(A=&MakeIntFromFloat(Float=$User.Variant), B=4, Comparison Type=::NewEnumerator4)
   ) {isSetParameters=true, module=SetVariables_#1}
  ]
  particleSpawnScript [
   &FN_InitializeAnimTrailBoneIndecies(First Bone or Socket Index=1)
   &FN_InitializeAnimTrail(
    AnimTrailWidth=1
    Color=&SelectLinearColorFromArray(
     Array Sampling Mode=::NewEnumerator1
     Direct Array Index=&MakeIntFromFloat(Float=$User.Variant)
     Color Selection Array=DataInterface() {propertyValues {ColorData [
        <rgba 0.0284 0.0284 0.0409 0.5>
        <rgba 0.9047 0.4852 0.7605 0.125>
        <rgba 0.0953 0.0723 0.0685 1>
        <rgba 0.1981 0.5583 0.6795 0.35>
        <rgba 0.5776 0.5906 0.5972 0.42>
        <rgba 0.0252 0.5647 0.7379 0.15>
       ]}}
     Random Seed=?
    )
    Lifetime=0.12
   )
  ]
  particleUpdateScript [
   &FN_UpdateAnimTrail_v2()
   &ScaleColor(
    Scale Alpha=&FloatFromCurve(
     FloatCurve=DataInterface() {propertyValues {Curve {keys [{interpMode=RCIM_Cubic, tangentMode=RCTM_User, time=0.25, value=1}, {interpMode=RCIM_Cubic, tangentMode=RCTM_User, time=1, value=0}]}}}
     Scale Curve=1.5
    )
    Scale RGB=&Multiply_VectorByFloat(
     Vector=&FN_Select_Vector_from_Time_Of_Day(Morning=<1 1 1>, Day=<1 1 1>, Evening=<1 1 1>, Night=<0.4 0.6 1>)
     Float=&FN_Select_Float_from_Time_Of_Day(Morning=1, Day=1, Evening=1, Night=0.65)
    )
    ScaleA=true
    ScaleRGB=true
   ) {module=ScaleColor002}
   &SolveForcesAndVelocity(Clamp Velocity=false, Limit Acceleration=false)
   /SetVariables() {isSetParameters=true, module=SetVariables_#2}
   &ScaleColor(
    Scale Alpha=&FloatFromCurve(
     FloatCurve=DataInterface() {propertyValues {Curve {keys [{interpMode=RCIM_Cubic, tangentMode=RCTM_User, time=0, value=1}, {interpMode=RCIM_Cubic, tangentMode=RCTM_User, time=1, value=0}]}}}
     CurveIndex=$Emitter.NormalizedLoopAge
    )
    ScaleA=true
    ScaleRGB=true
   )
  ]
  rendererClasses [&NiagaraRibbonRendererProperties]
  renderers [
   {rendererClass=&NiagaraRibbonRendererProperties}
   {rendererClass=&NiagaraRibbonRendererProperties, rendererIndex=1}
   {rendererClass=&NiagaraRibbonRendererProperties, rendererIndex=2}
   {rendererClass=&NiagaraRibbonRendererProperties, rendererIndex=3}
   {rendererClass=&NiagaraRibbonRendererProperties, rendererIndex=4}
   {rendererClass=&NiagaraRibbonRendererProperties, rendererIndex=5}
   {rendererClass=&NiagaraRibbonRendererProperties, rendererIndex=6}
  ]
 }
}
summary {
 systemName=NS_Pickaxe_CoralNumber_Trail
 emitters [
  {emitterName=Trail, rendererClasses [&NiagaraRibbonRendererProperties]}
  {emitterName=Trail001, rendererClasses [&NiagaraRibbonRendererProperties]}
 ]
 userVariables [
  User.bIsFrontEnd {type=NiagaraBool}
  User.bIsFrontEndPreview {type=NiagaraBool}
  User.bisSwinging {type=NiagaraBool}
  User.RibbonIDOffset {type=NiagaraInt32}
  User.SkeletalMesh {
   type=NiagaraDataInterfaceSkeletalMesh
   defaultValue {
    struct=NiagaraExt_VariableValue_DataInterface
    value {dataInterface=&self:NiagaraDataInterfaceSkeletalMesh_0}
   }
  }
  User.Variant {type=NiagaraFloat, defaultValue=0, description=""}
 ]
}
