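# Physics asset for the CoralNumber pickaxe axe mesh, giving its saw blade and blade base simple collision bodies and a joint.
# Each bone has one capsule: a tiny one on base_saw_blade and a larger one on saw_blade, offset along the bone's negative X axis.
# A single constraint joins them with twist and swing1 locked, leaving swing2 as the only unlocked angular axis for hinge-like motion.
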PhysicsAsset /BRCosmetics/Weapons/FORT_Melee/Pickaxe_CoralNumber/Meshes/CoralNumber_Axe_Physics
bodies {
 base_saw_blade [{shapeType=Capsule, length=0.1, radius=0.1, shapeName=base_saw_blade_capsule, rotation=<rot 0 -90 0>}]
 saw_blade [{shapeType=Capsule, length=2, radius=2, shapeName=saw_blade_capsule, center=<-30 0 0>}]
}
constraints [{
  bone1Name=saw_blade
  bone2Name=base_saw_blade
  swing1LimitDegrees=45
  swing1Motion=Locked
  swing2LimitDegrees=90
  twistLimitDegrees=45
  twistMotion=Locked
 }]
