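# Animation Blueprint for the CoralNumber pickaxe that spins its saw blade during melee swings and adds rigid-body physics secondary motion.
# In the front end it outputs a partly world-space rigid-body simulation; otherwise it chains an additive blade rotation into a handle-relative simulation.
# The event graph accumulates blade pitch while swinging and drives that simulation's interpolated alpha, clearing it after a short delay.
# It also computes a gravity override that the AnimGraph never reads.
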AnimBlueprint /BRCosmetics/Weapons/FORT_Melee/Pickaxe_CoralNumber/Meshes/CoralNumber_Axe_AnimBP
target_skeleton=&/CoralNumber_Axe_Skeleton
graphs [
 AnimGraph {nodes [
   n1 = RigidBody(ComponentPose=Local To Component(LocalPose=Use cached pose 'Input Pose Cache'())) {
    bOverrideWorldGravity=true
    overridePhysicsAsset=&/CoralNumber_Axe_Physics
    alphaBoolBlend {blendOption=Linear}
    baseBoneRef {boneName=b_AR_Root}
    componentAppliedLinearAccClamp=<0 3000 0>
    componentLinearAccScale {y=0.3}
    componentLinearVelScale {y=0.3}
    overrideWorldGravity {z=-980}
    simSpaceSettings {worldAlpha=0.3}
   }
   n2 = "Transform (Modify) Bone - Bone: base_saw_blade"(
    ComponentPose=Local To Component(LocalPose=Use cached pose 'Input Pose Cache'())
    Alpha=1
    Translation="0,0,0"
    Rotation=Get Rotator() -> (Rotator:Rotator)
    Scale="1,1,1"
   ) {
    rotationMode=BMM_Additive
    rotationSpace=BCS_BoneSpace
    translationMode=BMM_Additive
    translationSpace=BCS_BoneSpace
    alphaBoolBlend {blendOption=Linear}
    alphaScaleBiasClamp {interpSpeedDecreasing=2, interpSpeedIncreasing=2}
    boneToModify {boneName=base_saw_blade}
   }
   n3 = RigidBody(ComponentPose=@n2, Alpha=Get Alpha_IS_PlayingMelee_02() -> (Alpha_IS_PlayingMelee_02:double)) {
    bClampLinearTranslationLimitToRefPose=true
    bOverrideWorldGravity=true
    overridePhysicsAsset=&CoralNumber_Axe_Physics
    simulationSpace=BaseBoneSpace
    alphaBoolBlend {blendOption=Linear}
    alphaScaleBiasClamp {bInterpResult=true, interpSpeedDecreasing=1.5, interpSpeedIncreasing=5}
    baseBoneRef {boneName=Melee_TwoHanded_Handle}
    componentAppliedLinearAccClamp=<1500 0 1500>
    componentLinearAccScale {y=0.8}
    componentLinearVelScale {y=0.8}
   }
   Output Pose(Result=Blend Poses by bool(
     BlendPose_0=Component To Local(ComponentPose=@n1)
     BlendPose_1=Component To Local(ComponentPose=@n3)
     bActiveValue=Get Alpha_IS_FrontEnd() -> (Alpha_IS_FrontEnd:bool)
    ) {blendType=Linear})
   Save cached pose 'Input Pose Cache'(Pose=Input Pose()) {cacheName=Input Pose Cache}
  ]}
 EventGraph {nodes [
   n1 = Event BlueprintUpdateAnimation()
   > Set Delta Time X(Delta Time X=@n1.DeltaTimeX)
   > n2 = Sequence!(execute) -> (then_0, then_1, then_2:Exec, then_3:Exec, then_4:Exec)
   n3 = Branch(execute=@n2.then_1, Condition=Get bIsPlayingMeleeAnim())
   > Set Alpha_IS_PlayingMelee(Alpha_IS_PlayingMelee:double=1) -> (Output_Get:double)
   Set Alpha_IS_PlayingMelee(execute=@n3.else, Alpha_IS_PlayingMelee=0)
   n4 = Gravity Override State Function DEPRECATED(
    execute=@n2.then_0
    CharacterPartAnimInstance=Self-Reference()
    GameTimeInSeconds=Get Delta Time X()
    GravityOverrideParams=Get Gravity Override Params()
   )
   > Set GravityOverride(GravityOverride=@n4)
   n5 = Branch(execute=@n2.then_3, Condition=Get bIsPlayingMeleeAnim())
   > Set Alpha_IS_PlayingMelee_02(Alpha_IS_PlayingMelee_02:double=1) -> (Output_Get:double)
   Delay(execute=@n5.else, Duration=0.3)
   > Set Alpha_IS_PlayingMelee_02(Alpha_IS_PlayingMelee_02=0)
   n6 = Branch(execute=@n2.then_4, Condition=Get bIsInFrontEnd())
   > Set Alpha_IS_FrontEnd(Alpha_IS_FrontEnd:bool=1) -> (Output_Get:bool)
   Set Alpha_IS_FrontEnd(execute=@n6.else, Alpha_IS_FrontEnd=false)
   n7 = float * float(
    A=Get Alpha_IS_PlayingMelee() -> (Alpha_IS_PlayingMelee:double)
    B=Select(Option 0:double=0, Option 1:double=470, Index=Get bIsPlayingMeleeAnim()) -> (ReturnValue:double)
   ) {is=title K2Node_PromotableOperator}
   n8 = Get Delta Time X()
   n9 = Set current_velocity(
    execute=@n2.then_2
    current_velocity:double=float + float(A=float * float(A=@n8, B=@n7), B=Get current_velocity() -> (current_velocity:double))
   ) -> (Output_Get:double)
   > Set Rotator(Rotator:Rotator=MakeRotator(Pitch=@n9)) -> (Output_Get:Rotator)
  ]}
]
variables [
 Alpha_IS_FrontEnd {type=boolean}
 Alpha_IS_PlayingMelee {type=number}
 Alpha_IS_PlayingMelee_02 {type=number}
 current_velocity {type=number, default=300}
 Delta Time X {type=number}
 Gravity Override Params {
  type=GravityOverrideParamsStruct
  default {onGround_Standing {gravityOverride=<0 0 -500>}, skydive_GoingUp {gravityOverride=<0 0 -980>, windAmplitude=<0 0 0>, windFrequency=<0 0 0>}}
 }
 GravityOverride {type=Vector}
 Rotator {type=Rotator}
 RotatorValue {type=number}
 strap scale {type=number}
]
